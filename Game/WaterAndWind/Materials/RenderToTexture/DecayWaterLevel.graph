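# DecayWaterLevel is an unlit translucent material in the water and wind render-to-texture set, apparently for fading a water level over time.
# It multiplies a DecayRate scalar by a TickDelta scalar to form a per-tick decay amount.
# A FortniteMaterialParameters collection node labelled tick delta sits apart from that multiply.
# The graph shows no connection from these nodes to a material output.
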Material /Game/WaterAndWind/Materials/RenderToTexture/DecayWaterLevel
graph {
 properties {bUsedWithNanite=true, blendMode=BLEND_Translucent, pixelDepthOffsetMode=PDOM_Legacy, shadingModel=MSM_Unlit, usedShadingModels=Unlit}
 nodes [
  CollectionParameter() {collection=&FortniteMaterialParameters, desc=tick delta}
  Multiply(A=ScalarParameter() {defaultValue=0.25, parameterName=DecayRate}, B=ScalarParameter() {defaultValue=0.25, parameterName=TickDelta})
 ]
}
