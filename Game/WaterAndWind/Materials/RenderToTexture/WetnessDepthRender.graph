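# Translucent unlit material that, judging by its name and folder, renders player wetness depth into a texture as a vertical gradient.
# It samples wind-water world data at the Player Location parameter and normalizes water height above the player by the collection's player height, clamped.
# UV V is compared to that level, filling values below it and adding a soft SphereMask edge.
# The scaled result is intended to build up opacity over time.
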Material /Game/WaterAndWind/Materials/RenderToTexture/WetnessDepthRender
graph {
 properties {bUsedWithNanite=true, blendMode=BLEND_Translucent, pixelDepthOffsetMode=PDOM_Legacy, shadingModel=MSM_Unlit, usedShadingModels=Unlit}
 nodes [
  Constant() {r=1}
  n1 = MaterialFunctionCall(Float2=TextureCoordinate()) {materialFunction=&BreakOutFloat2Components}
  n2 = VectorParameter() {parameterName=Player Location, defaultValue=<rgba 0 0 0 1>}
  n3 = MaterialFunctionCall!(Position=@n2.RGB) {materialFunction=/Game/WaterAndWind/MaterialFunctions/Fortnite_SampleWindWaterIsometricWorldData}
  n4 = CollectionParameter() {
   collection=&FortniteMaterialParameters
   parameterName=PlayerHeight_DoNotSetDynamically
  }
  n5 = Clamp(None=Divide(A=Subtract(A=@n3.Water Z height, B=@n2.B), B=@n4)) {desc=if the water is under the terrain then we need to cap its max value}
  n6 = SphereMask(A=@n1.G, B=@n5) {attenuationRadius=0.2, hardnessPercent=0}
  n7 = If(A=@n1.G, B=@n5, "A > B"=Constant(), "A < B"=Constant() {r=1})
  Multiply(A=Clamp(None=Add(A=@n6, B=@n7))) {constB=0.85, desc=build up opacity over time}
 ]
}
